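; Actor Blueprint for the Artemis dark pine foliage variant 02, with a function to randomize its scale.
; ResetScale reads the RandomScaleRange Vector2D as minimum and maximum, then sets the actor's 3D scale from three separate random draws, so the axes can scale non-uniformly.
; The construction script is empty, and nothing in this file calls ResetScale.
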(fn ResetScale
  (bind (_x _y) (Math|Vector2D|BreakVector2D (Variables|GetRandomScaleRange)))
  (Transformation|SetActorScale3D self (Math|Vector|MakeVector (Math|Random|RandomFloatinRange _x _y) (Math|Random|RandomFloatinRange _x _y) (Math|Random|RandomFloatinRange _x _y))))

(fn ConstructionScript)
